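# Sound cue for a stunt ramp's fire effect in LootCannon, layering pyro explosions, firework bursts and fire pit audio with attenuation.
# A Mixer plays a random, pitch-varied explosion immediately, plus staggered delayed repeats that pass through an extra pitch-up modulator.
# Separately delayed branches add a random fire pit loop clip and a random firework burst, each shaped by an Enveloper with a quick pitch drop and volume fade.
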SoundCue /LootCannon/Gameplay/Ramp/Audio/SC_RampFire
root=Mixer_0
nodes [
 {id=Delay_0, children [Enveloper_1], properties {delayMax=0.3, delayMin=0.3}}
 {id=Delay_2, children [Modulator_1], properties {delayMax=0.3, delayMin=0.3}}
 {id=Delay_3, children [Modulator_1], properties {delayMax=0.5, delayMin=0.5}}
 {id=Delay_4, children [Modulator_1], properties {delayMax=0.7, delayMin=0.7}}
 {id=Delay_5, children [Enveloper_2], properties {delayMax=0.6, delayMin=0.6}}
 {
  id=Enveloper_1
  children [Modulator_3]
  properties {
   pitchCurve {editorCurveData {keys [<at -0.0003 4.6835>,<at 0.0602 2.2389>]}}
   volumeCurve {editorCurveData {keys [<at 0.0153 1.4968>,<at 1.0281 0.8209>,<at 2.5934 -0.007>]}}
  }
 }
 {
  id=Enveloper_2
  children [Modulator_5]
  properties {
   pitchCurve {editorCurveData {keys [<at -0.0003 4.6835>,<at 0.0602 2.2389>]}}
   volumeCurve {editorCurveData {keys [<at 0.0153 1.4968>,<at 1.0281 0.8209>,<at 2.5933 -0.007>]}}
  }
 }
 {id=Mixer_0, children [null,Modulator_2,Delay_2,Delay_3,Delay_4,Delay_0,Delay_5], properties {inputVolume [0.5,2,1.8,1.6,1.3,1,2.1]}}
 {id=Modulator_1, children [Modulator_2], properties {pitchMax=2.5, pitchMin=2, volumeMax=1, volumeMin=1}}
 {id=Modulator_2, children [Random_2], properties {pitchMax=1.5, pitchMin=1, volumeMax=1, volumeMin=1}}
 {id=Modulator_3, children [Random_3], properties {pitchMin=0.98, volumeMax=1, volumeMin=1}}
 {id=Modulator_5, children [Random_1], properties {pitchMin=0.98, volumeMax=1, volumeMin=1}}
 {
  id=Random_1
  children [WavePlayer_1, WavePlayer_15, WavePlayer_16, WavePlayer_18]
  properties {bShouldExcludeFromBranchCulling=true, weights [1,1,1,1]}
 }
 {
  id=Random_2
  children [WavePlayer_8, WavePlayer_9, WavePlayer_13, WavePlayer_11, WavePlayer_12]
  properties {bShouldExcludeFromBranchCulling=true, weights [1,1,1,1,1]}
 }
 {id=Random_3, children [WavePlayer_22, WavePlayer_23], properties {bShouldExcludeFromBranchCulling=true, weights [1,1]}}
 {id=WavePlayer_1, properties {soundWaveAssetPtr=/Game/Sounds/Helios/POI/NitroDrome/Fireworks/NitroDrome_Fireworks_Burst_2D_01}}
 {id=WavePlayer_11, properties {soundWaveAssetPtr=/LootCannon/Gameplay/Ramp/Audio/SoundWaves/StuntRamp_Pyro_Explosion_04}}
 {id=WavePlayer_12, properties {soundWaveAssetPtr=/LootCannon/Gameplay/Ramp/Audio/SoundWaves/StuntRamp_Pyro_Explosion_05}}
 {id=WavePlayer_13, properties {soundWaveAssetPtr=/LootCannon/Gameplay/Ramp/Audio/SoundWaves/StuntRamp_Pyro_Explosion_03}}
 {id=WavePlayer_15, properties {soundWaveAssetPtr=/Game/Sounds/Helios/POI/NitroDrome/Fireworks/NitroDrome_Fireworks_Burst_2D_08}}
 {id=WavePlayer_16, properties {soundWaveAssetPtr=&NitroDrome_Fireworks_Burst_2D_08}}
 {id=WavePlayer_18, properties {soundWaveAssetPtr=/Game/Sounds/Helios/POI/NitroDrome/Fireworks/NitroDrome_Fireworks_Burst_2D_11}}
 {id=WavePlayer_22, properties {soundWaveAssetPtr=/LootCannon/Gameplay/Ramp/Audio/SoundWaves/FirePitHazard_Amb_Loop_04}}
 {id=WavePlayer_23, properties {soundWaveAssetPtr=/LootCannon/Gameplay/Ramp/Audio/SoundWaves/FirePitHazard_Amb_Loop_03}}
 {id=WavePlayer_8, properties {soundWaveAssetPtr=/LootCannon/Gameplay/Ramp/Audio/SoundWaves/StuntRamp_Pyro_Explosion_01}}
 {id=WavePlayer_9, properties {soundWaveAssetPtr=/LootCannon/Gameplay/Ramp/Audio/SoundWaves/StuntRamp_Pyro_Explosion_02}}
]
properties {
 attenuationSettings=&/Ramp_Fire_Att
 maxDistance=8000
 soundClassObject=&Fort_Ambient_Athena
 volumeMultiplier=1.3
}
