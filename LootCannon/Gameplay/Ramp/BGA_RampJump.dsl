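; Ramp jump challenge actor that rewards an actor, intended as a vehicle, for passing TopBox and then LowerBox.
; On authority, a top overlap records the actor, triggers remote-side Niagara and sound, and after a timed delay rewards a matching bottom overlap, then calls ResetRamp.
; The one-time reward fires referenced loot cannons, with a distinct tier for the first, then destroys them and the ramp.
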(event Custom|ResetRamp
  (Utilities|FlowControl|DoOnce false
    (:Completed
      (Variables|Default|SetVehicleActorTop (stale OnComponentBeginOverlap(TopBox) OtherActor))
      (CallFunction|TopCosmetics)
      (Utilities|FlowControl|Delay 4.75)
      (Utilities|IsValid (Variables|Default|GetVehicleActorBottom)
        (:Is%20Valid
          (if (== (Variables|Default|GetVehicleActorTop) (Variables|Default|GetVehicleActorBottom))
            (CallFunction|GiveRewards)
            (else
              _)))
        (:Is%20Not%20Valid
          _))
      (CallFunction|ResetRamp))))

(event OnComponentBeginOverlap(TopBox)
  (params (OverlappedComponent :type Primitive%20Component%20Object%20Reference) (OtherActor :type Actor%20Object%20Reference) (OtherComp :type Primitive%20Component%20Object%20Reference) (OtherBodyIndex :type Integer) (bFromSweep :type Boolean) (SweepResult :type Hit%20Result%20Structure%20%28by%20ref%29))
  (Networking|SwitchHasAuthority
    (:Authority
      (Utilities|FlowControl|DoOnce false
        (:Completed
          (Variables|Default|SetVehicleActorTop OtherActor)
          (CallFunction|TopCosmetics)
          (Utilities|FlowControl|Delay 4.75)
          (Utilities|IsValid (Variables|Default|GetVehicleActorBottom)
            (:Is%20Valid
              (if (== (Variables|Default|GetVehicleActorTop) (Variables|Default|GetVehicleActorBottom))
                (CallFunction|GiveRewards)
                (else
                  _)))
            (:Is%20Not%20Valid
              _))
          (CallFunction|ResetRamp))))))

(event OnComponentBeginOverlap(LowerBox)
  (params (OverlappedComponent :type Primitive%20Component%20Object%20Reference) (OtherActor :type Actor%20Object%20Reference) (OtherComp :type Primitive%20Component%20Object%20Reference) (OtherBodyIndex :type Integer) (bFromSweep :type Boolean) (SweepResult :type Hit%20Result%20Structure%20%28by%20ref%29))
  (Networking|SwitchHasAuthority
    (:Authority
      (if (== OtherActor (Variables|Default|GetVehicleActorTop))
        (Variables|Default|SetVehicleActorBottom OtherActor) ; only set this if the player was at the top first
      ))))

(event Custom|TopCosmetics%20
  (bind _niagara_top_a (Variables|Default|GetNiagara_Top_A))
  (Networking|SwitchHasAuthority
    (:Remote
      (Components|Activation|Activate _niagara_top_a false)
      (Audio|SpawnSoundatLocation "/LootCannon/Gameplay/Ramp/Audio/SC_RampFire.SC_RampFire" (Transformation|GetWorldLocation _niagara_top_a)))))

(event Custom|GiveRewards
  (Networking|SwitchHasAuthority
    (:Authority
      (Utilities|FlowControl|DoOnce false
        (:Completed
          (for (_array_element _array_index) (Variables|Default|GetLootCannonReferences)
            (Utilities|IsValid _array_element
              (:Is%20Valid
                (if (== _array_index 0)
                  (Variables|Default|SetLoot%20Tier%20to%20Fire "LTG_GreatRamp")
                  (else
                    (Variables|Default|SetLoot%20Tier%20to%20Fire "Loot_ApolloTreasure_Rare")))
                (Class|BGALootCannon|FireLootCannon _array_element (Variables|Default|GetLoot%20Tier%20to%20Fire) 0))))
          (Utilities|FlowControl|Delay 2.0)
          (for _array_element_2 (Variables|Default|GetLootCannonReferences)
            (Utilities|IsValid _array_element_2
              (:Is%20Valid
                (Actor|DestroyActor _array_element_2))))
          (Utilities|FlowControl|Delay 5.0)
          (Actor|DestroyActor self))))
    (:Remote
      (Utilities|FlowControl|DoOnce false))))

(fn ConstructionScript)
